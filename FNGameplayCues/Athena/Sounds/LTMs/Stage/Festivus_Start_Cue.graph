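# Festivus_Start_Cue is a Sound Cue filed under the Athena LTM Stage sounds, named for the start of Festivus.
# It has no root or sound nodes, so the cue itself produces no audio.
# It keeps only cue-level settings: the scripted-events sound class, a reduced volume multiplier and a very large maximum distance.
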SoundCue /FNGameplayCues/Athena/Sounds/LTMs/Stage/Festivus_Start_Cue
root=null
nodes []
properties {maxDistance=2097152, soundClassObject=/Game/Sounds/FortSoundClassesAndModes/Fort_Events_Scripted, volumeMultiplier=0.4211}
